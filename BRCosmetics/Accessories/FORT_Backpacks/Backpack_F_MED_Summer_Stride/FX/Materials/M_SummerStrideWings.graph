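# Additive, unlit, two-sided Niagara material for the glowing wings of the Summer Stride backpack.
# A base-to-tip gradient blends toward a deep color through panning scanlines, with an optional chromatic overlay.
# Opacity combines scanlines, particle alpha, depth fade and a vertex-red WingReveal wipe.
# World offset adds vertex-red-varied sine wobble to lagged rotation, a sign-corrected, normalized blend of two dynamic-parameter quaternions weighted by vertex red.
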Material /BRCosmetics/Accessories/FORT_Backpacks/Backpack_F_MED_Summer_Stride/FX/Materials/M_SummerStrideWings
graph {
 properties {
  bUsedWithNiagaraMeshParticles=true
  bUsedWithNiagaraRibbons=true
  bUsedWithNiagaraSprites=true
  blendMode=BLEND_Additive
  pixelDepthOffsetMode=PDOM_Legacy
  shadingModel=MSM_Unlit
  twoSided=true
  usedShadingModels=Unlit
 }
 nodes [
  Add(A=NamedRerouteUsage(), B=NamedRerouteUsage())
  n1 = ComponentMask(None=TextureCoordinate()) {g=true}
  n2 = MaterialFunctionCall() {materialFunction=&LocalPosition}
  n3 = TextureSample(UVs=MaterialFunctionCall!(
    UV=ComponentMask(None=@n2."Instance Local Position (Excluding Offsets)") {b=true, r=true}
    Tile U=ScalarParameter() {defaultValue=1, parameterName="Tile U (S)"}
    Tile V=ScalarParameter() {defaultValue=0.3, parameterName="Tile V (S)"}
    Pan U=ScalarParameter() {defaultValue=-2, parameterName="Pan U (S)"}
    Pan V=ScalarParameter() {defaultValue=3.72, parameterName="Pan V (S)"}
    Offset U
    Offset V
    Time
   ) {materialFunction=/Game/Effects/Fort_Effects/Materials/_Functions/MF_PanTileOffset}.UV's) {texture=/Game/PvP/StrategicBuildingActors/Defensive/LaserFence/Assets/LaserScanline}
  n4 = LinearInterpolate(
   A=VectorParameter() {parameterName=Deep Color, defaultValue=<rgba 0.0086 0.327 0.5 0>}.RGB
   B=LinearInterpolate(
    A=VectorParameter() {parameterName=Base Color, defaultValue=<rgba 0.0015 0.1861 0.2 1>}.RGB
    B=VectorParameter() {parameterName=Tip Color, defaultValue=<rgba 0.509 1.8205 2 1>}.RGB
    Alpha=OneMinus(None=@n1)
   )
   Alpha=@n3.G
  )
  n5 = TextureSample(UVs=MaterialFunctionCall(
    Tile U=ScalarParameter() {defaultValue=1, parameterName=Tile U 2}
    Tile V=ScalarParameter() {defaultValue=1, parameterName=Tile V 2}
    Pan U=ScalarParameter() {defaultValue=1, parameterName=Pan U 2}
    Pan V=ScalarParameter() {defaultValue=2, parameterName=Pan V 2}
   ) {materialFunction=&MF_PanTileOffset}.UV's) {texture=/Game/Effects/Fort_Effects/Textures/Tile/T_Smooth_Tile_Chromatic}
  n6 = DotProduct(A=@n5.RGB, B=Constant3Vector() {constant=<rgba 0.2 0.2 0.2 0>})
  n7 = LinearInterpolate(A=@n6, B=@n5.RGB, Alpha=ScalarParameter() {defaultValue=5, parameterName=OverSat}) {constAlpha=15}
  n8 = Multiply(A=@n6, B=ScalarParameter() {parameterName=OverSatPow})
  MaterialFunctionCall(Emissive=Multiply(A=LinearInterpolate(A=@n4, B=@n7, Alpha=@n8), B=ParticleColor().RGB)) {materialFunction=&MF_FX_EmissiveAdjustments}
  n9 = ParticleColor()
  n10 = LinearInterpolate(A=@n3.G) {constAlpha=0.1}
  Multiply(A=Multiply(A=Multiply(A=@n9.A, B=@n10), B=NamedRerouteUsage()), B=DepthFade() {fadeDistanceDefault=10})
  n11 = DynamicParameter() {paramNames [WingReveal, RotationAngle, Param3, Param4]}
  n12 = Add(A=@n11.WingReveal) {constB=0.2}
  n13 = VertexColor()
  NamedRerouteDeclaration(None=OneMinus(None=SmoothStep(Min=@n11.WingReveal, Max=@n12, Value=@n13.R))) {name=RevealMask, nodeColor=<rgba 0.2187 1 0 1>}
  n14 = Time() {bOverride_Period=true, period=5000}
  n15 = VertexColor()
  n16 = LinearInterpolate(
   A=ScalarParameter() {defaultValue=1, parameterName=WobbleSpeed}
   B=ScalarParameter() {defaultValue=1.5, parameterName=WobbleSpeedHigh}
   Alpha=Frac(None=Multiply(A=@n15.R) {constB=4})
  )
  n17 = Sine(None=Add(A=Multiply(A=@n14, B=@n16), B=Multiply(A=@n15.R) {constB=13}))
  n18 = Transform(None=Constant3Vector() {constant=<rgba 0 0 1 0>}) {transformSourceType=TRANSFORMSOURCE_Local}
  NamedRerouteDeclaration(None=Multiply(A=Multiply(A=@n17, B=@n18), B=ScalarParameter() {defaultValue=0.3, parameterName=WobbleAmount})) {name=Wobble, nodeColor=<rgba 0.2656 0 1 1>}
  n19 = DynamicParameter() {parameterIndex=1, defaultValue=<rgba 0 0 0 1>, paramNames [RotQuatX,RotQuatY,RotQuatZ,RotQuatW]}
  n20 = MaterialFunctionCall(R=@n19.RotQuatX, G=@n19.RotQuatX, B=@n19.RotQuatX, A=@n19.RotQuatX) {materialFunction=&AppendMany}
  n21 = OneMinus(None=VertexColor().R)
  n22 = DynamicParameter() {parameterIndex=2, paramNames [MeshRotX,MeshRotY,MeshRotZ,MeshRotW]}
  n23 = MaterialFunctionCall(R=@n22.MeshRotX, G=@n22.MeshRotX, B=@n22.MeshRotX, A=@n22.MeshRotX) {materialFunction=&AppendMany}
  n24 = DotProduct(A=@n20.RGBA, B=@n23.RGBA)
  n25 = Constant()
  n26 = Multiply(A=@n21) {constB=-1}
  n27 = Add(A=Multiply(A=@n20.RGBA, B=OneMinus(None=@n21)), B=Multiply(A=If(A=@n24, B=@n25, "A > B"=@n21, "A == B"=@n26, "A < B"=@n26), B=@n23.RGBA))
  NamedRerouteDeclaration(None=MaterialFunctionCall!(Quaternion=Normalize(VectorInput=@n27), Invert Quaternion=StaticBool(), World Position, Particle Position) {materialFunction=/Niagara/Functions/DataPacking/Niagara_RotateMeshParticleWithQuaternion}.Result) {name=LaggedRotation, nodeColor=<rgba 1 0.8672 0 1>}
 ]
}
