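# Niagara system for the Summer Stride back bling, combining mirrored mesh wings and spinning rays with drifting mesh globs, a low glow sprite and two lights.
# System.NormalizedTimeSinceLastEvent drives curves for wing material, ray scale, color and spin, glob color and size, and light radius, so the effect flares over an event.
# Wings use interpolated owner rotation to lag behind movement.
# User parameters expose event duration and a kill flag.
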NiagaraSystem /BRCosmetics/Accessories/FORT_Backpacks/Backpack_F_MED_Summer_Stride/FX/NS_SummerStrideBackpack
dependencies {
 usedDataInterfaces [&NiagaraDataInterfaceColorCurve, &NiagaraDataInterfaceCurve, &NiagaraDataInterfaceVectorCurve]
 usedDynamicInputs [
  &Boolean_NotOperation
  &MakeCustomFloatFromBool
  &SetBoolByFloatComparison
  &ClampFloat
  &Add_Float
  &Divide_Float
 ]
 usedModules [
  &self:SystemSpawnScript.SystemScriptSource.SystemScriptGraph.NiagaraNodeAssignment_0.SetVariables_#1
  &self:SystemSpawnScript.SystemScriptSource.SystemScriptGraph.NiagaraNodeAssignment_4.SetVariables_#2
  &SystemState
  &DoOnce
  &self:SystemSpawnScript.SystemScriptSource.SystemScriptGraph.NiagaraNodeAssignment_6.SetVariables_#3
  &self:SystemSpawnScript.SystemScriptSource.SystemScriptGraph.NiagaraNodeAssignment_7.SetVariables_#4
  &self:SystemSpawnScript.SystemScriptSource.SystemScriptGraph.NiagaraNodeAssignment_1.SetVariables_#5
 ]
}
emitters {
 Lights {
  emitterUpdateScript [
   &EmitterState(
    Inactive Response=::NewEnumerator1
    Life Cycle Mode=::NewEnumerator1
    Loop Behavior=::NewEnumerator1
    Loop Duration Mode=::NewEnumerator1
    UseLoopDelay=false
   )
   &SpawnBurst_Instantaneous(Spawn Count=2, Use Spawn Probability=false, UseLoopCountLimit=false)
  ]
  particleSpawnScript [
   &FN_SetSpawnGroup(Number of Spawn Groups=2)
   &InitializeParticle(
    Color Mode=::NewEnumerator0
    Lifetime=2
    Position Offset=&MaskVectorBySpawnGroup(Masked Vector=<20 0 11>, Passthrough Vector=<-25 0 11>)
    UsePositionOffset=true
    Write Material Random=true
   )
  ]
  particleUpdateScript [
   &ParticleState(Kill Particles When Lifetime Has Elapsed=false, Let Infinitely Lived Particles Die When Emitter Deactivates=true)
   &FN_Light_Attributes(
    Color=<rgba 0.0135 0.42 1 1>
    Create Color Attribute=true
    Create Exponent Attribute=true
    Create Radius Attribute=true
    Exponent=4
    Radius=&FloatFromCurve(
     FloatCurve=DataInterface() {propertyValues {Curve {keys [
         {arriveTangent=3.0388, interpMode=RCIM_Cubic, leaveTangent=3.0388, tangentMode=RCTM_User, time=0, value=-0.1}
         {arriveTangent=1.6667, leaveTangent=1.6667, time=0.15, value=1}
         <at 0.85 1 RCIM_Cubic>
         <at 1 -0.1>
        ]}}}
     CurveIndex=$System.NormalizedTimeSinceLastEvent
     Scale Curve=50
    )
    Scale Color=2
   )
  ]
  rendererClasses [&NiagaraLightRendererProperties]
  renderers [{rendererClass=&NiagaraLightRendererProperties, rendererIndex=0}]
 }
 LowGlow {
  emitterUpdateScript [
   &EmitterState(Life Cycle Mode=::NewEnumerator1, Loop Behavior=::NewEnumerator1, Loop Duration Mode=::NewEnumerator1, UseLoopDelay=false)
   &SpawnBurst_Instantaneous(Use Spawn Probability=false, UseLoopCountLimit=false)
  ]
  particleSpawnScript [&InitializeParticle(Color=<rgba 0.6666 1 0.994 0.1>, Lifetime=5, Position Offset=<0 0 -9>, UsePositionOffset=true, Write Material Random=true)]
  particleUpdateScript [&ParticleState(Kill Particles When Lifetime Has Elapsed=false, Let Infinitely Lived Particles Die When Emitter Deactivates=true)]
  rendererClasses [&NiagaraSpriteRendererProperties]
  renderers [{rendererClass=&NiagaraSpriteRendererProperties}]
 }
 NE_SummerStrideGlobs {
  emitterUpdateScript [
   &EmitterState()
   &SpawnRate(SpawnRate=3, Use Spawn Probability=false)
  ]
  particleSpawnScript [
   &InitializeParticle(
    Lifetime Max=3
    Lifetime Min=2
    Lifetime Mode=::NewEnumerator1
    Mesh Scale Mode=::NewEnumerator1
    Mesh Uniform Scale Max=0.4
    Mesh Uniform Scale Min=0.25
    Sprite Size Mode=::NewEnumerator3
    Uniform Sprite Size=1
    UsePositionOffset=false
    Write Material Random=true
   )
   &ShapeLocation(
    MaskSpawnGroup=false
    Offset=<0 -3 -9>
    Offset Mode=::NewEnumerator0
    Override Seed=false
    Shape Origin=/Niagara/DynamicInputs/Helpers/SimulationPosition()
    Sphere Radius=2
    UseOwnerScale=true
   )
   &InitialMeshOrientation(Use Rotation Vector=false)
   &AddVelocity(Velocity=&RandomRangeVector(Minimum=<3 0 -10>, Maximum=<-3 0 -15>))
  ]
  particleUpdateScript [
   &ParticleState()
   &Drag(Use Linear Drag=true, Use Rotational Drag=false)
   &Color(Color=&ColorFromCurve(
     ColorCurve=DataInterface() {propertyValues {
       AlphaCurve {keys [<at 0 1 RCIM_Cubic>]}
       BlueCurve {keys [<at 0 0 RCIM_Cubic>,<at 1 0.1644 RCIM_Cubic>]}
       GreenCurve {keys [<at 0 0 RCIM_Cubic>,<at 1 1 RCIM_Cubic>]}
       RedCurve {keys [<at 0 1 RCIM_Cubic>,<at 1 0 RCIM_Cubic>]}
      }}
     CurveIndex=$System.NormalizedTimeSinceLastEvent
    ))
   &ScaleMeshSize(Scale Factor=&VectorFromFloat(Value=&FloatFromCurve(
      FloatCurve=DataInterface() {propertyValues {Curve {keys [
          {arriveTangent=9.0725, interpMode=RCIM_Cubic, leaveTangent=9.0725, tangentMode=RCTM_User, time=0, value=0}
          {arriveTangent=1.0799, interpMode=RCIM_Cubic, leaveTangent=1.0799, tangentMode=RCTM_User, time=0.1408, value=0.9847}
          {arriveTangent=-4.6925, interpMode=RCIM_Cubic, leaveTangent=-4.6925, tangentMode=RCTM_User, time=1, value=0}
         ]}}}
      Scale Curve=&FloatFromCurve(
       FloatCurve=DataInterface() {propertyValues {Curve {keys [
           {arriveTangent=2.338, interpMode=RCIM_Cubic, leaveTangent=2.338, tangentMode=RCTM_User, time=0, value=0.7}
           <at 0.2 1>
           <at 0.8 1 RCIM_Cubic>
           {arriveTangent=-2.0675, interpMode=RCIM_Cubic, leaveTangent=-2.0675, tangentMode=RCTM_User, time=1, value=0.7}
          ]}}}
       CurveIndex=$System.NormalizedTimeSinceLastEvent
      )
     )))
   &PointAttractionForce(
    Attraction Strength=15
    Attractor Position Offset:Vector3f=<0 -2 -6>
    Kill Within Radius=false
    Module.Use Falloff=true
    UseAttractorOffset=true
   )
   &SolveForcesAndVelocity(Clamp Velocity=false, Limit Acceleration=false)
  ]
  rendererClasses [&NiagaraMeshRendererProperties]
  renderers [{rendererClass=&NiagaraMeshRendererProperties, rendererIndex=0}]
 }
 Rays {
  emitterUpdateScript [
   &EmitterState(
    Inactive Response=::NewEnumerator1
    Life Cycle Mode=::NewEnumerator1
    Loop Behavior=::NewEnumerator1
    Loop Duration Mode=::NewEnumerator1
    UseLoopDelay=false
   )
   &SpawnBurst_Instantaneous(Use Spawn Probability=false, UseLoopCountLimit=false)
   &SpawnBurst_Instantaneous(Spawn Group=1, Use Spawn Probability=false, UseLoopCountLimit=false) {module=SpawnBurst_Instantaneous001}
  ]
  particleSpawnScript [
   &InitializeParticle(
    Color=<rgba 0.6458 0.9631 1 0.5>
    Lifetime=2
    Mesh Scale=<0.23 0.25 0.14>
    Mesh Scale Mode=::NewEnumerator4
    Position Offset=&MaskVectorBySpawnGroup(Masked Vector=<5 -3 0>, Passthrough Vector=<-5 -3 0>)
    Sprite Size Mode=::NewEnumerator3
    Uniform Sprite Size=25
    UsePositionOffset=true
    Write Material Random=true
   )
   &InitialMeshOrientation(
    Mesh Orientation Mode=::NewEnumerator2
    Orientation Vector=<0 0 -1>
    Rotation=&MaskVectorBySpawnGroup(Masked Vector=<-0.03 -0.06 0>, Passthrough Vector=<0.53 0.06 -0.05>)
    Use Rotation Vector=true
   )
   /SetVariables(Particles.SpriteCameraOffset=3) {isSetParameters=true}
   /SetVariables(Particles.SpritePosition:NiagaraPosition=&Add_Vector(
     A=Linked() {linkedVariable {name=Particles.Position, type=NiagaraPosition}}
     B=&MaskVectorBySpawnGroup(Masked Vector=<1 0 0>, Passthrough Vector=<-1 0 0>)
    )) {isSetParameters=true}
  ]
  particleUpdateScript [
   &ParticleState(Kill Particles When Lifetime Has Elapsed=false, Let Infinitely Lived Particles Die When Emitter Deactivates=true)
   &UpdateMeshOrientation(
    Rotation Rate=&Multiply_Float(
     A=&MaskFloatBySpawnGroup(Masked Float=-0.5, Passthrough Float=0.5)
     B=&FloatFromCurve(
      FloatCurve=DataInterface() {propertyValues {Curve {keys [<at 0 1 RCIM_Cubic>,<at 0.5 1.5 RCIM_Cubic>,<at 1 1>]}}}
      CurveIndex=$System.NormalizedTimeSinceLastEvent
     )
    )
    Rotation Vector=<0 0 1>
   )
   &ScaleMeshSize(Scale Factor=&VectorFromCurve(
     VectorCurve=DataInterface() {propertyValues {
       XCurve {keys [<at 0 1>, <at 0.2 1.5>, <at 0.5 1.5>, {arriveTangent=-6.2682, leaveTangent=-6.2682, time=1, value=1}]}
       YCurve {keys [<at 0 1>, <at 0.2 1.5>, <at 0.5 1.5>, {arriveTangent=-6.2682, leaveTangent=-6.2682, time=1, value=1}]}
       ZCurve {keys [
         {arriveTangent=10.1392, interpMode=RCIM_Cubic, leaveTangent=10.1392, tangentMode=RCTM_User, time=0, value=0.5}
         <at 0.2 2>
         <at 0.5 2 RCIM_Cubic>
         {arriveTangent=-5.9432, interpMode=RCIM_Cubic, leaveTangent=-5.9432, tangentMode=RCTM_User, time=1, value=0.5}
        ]}
      }}
     CurveIndex=$System.NormalizedTimeSinceLastEvent
    ))
   &ScaleColor(
    Curve Index=$System.NormalizedTimeSinceLastEvent
    Linear Color Curve=DataInterface() {propertyValues {
      AlphaCurve {keys [<at 0 0.5>,<at 0.5 1>,<at 1 0.5>]}
      BlueCurve {keys [<at 0 1>,<at 0.2 3>,<at 0.25 1>,<at 1 1>]}
      GreenCurve {keys [<at 0 1>,<at 0.2 3>,<at 0.25 1>,<at 1 1>]}
      RedCurve {keys [<at 0 0.5>,<at 0.2 3>,<at 0.25 1>,<at 1 0.5>]}
     }}
    Scale Mode=::NewEnumerator2
   )
   &SpriteRotationRate(Rotation Rate=&Multiply_Float(
     A=&MaskFloatBySpawnGroup(Masked Float=-360, Passthrough Float=360)
     B=&FloatFromCurve(
      FloatCurve=DataInterface() {propertyValues {Curve {keys [<at 0 0.75 RCIM_Cubic>,<at 0.5 1.5 RCIM_Cubic>,<at 1 0.75>]}}}
      CurveIndex=$System.NormalizedTimeSinceLastEvent
     )
    ))
   &ScaleSpriteSize(
    Uniform Curve Index=$System.NormalizedTimeSinceLastEvent
    Uniform Curve Sprite Scale=DataInterface() {propertyValues {Curve {keys [
        <at 0 1 RCIM_Cubic>
        {arriveTangent=-3.4395, interpMode=RCIM_Cubic, leaveTangent=-3.4395, tangentMode=RCTM_User, time=0.2, value=2}
        <at 0.25 1.1>
        {tangentMode=RCTM_User, time=1, value=1}
       ]}}}
   )
  ]
  rendererClasses [&NiagaraMeshRendererProperties, &NiagaraSpriteRendererProperties]
  renderers [
   {rendererClass=&NiagaraMeshRendererProperties, rendererIndex=0}
   {rendererClass=&NiagaraSpriteRendererProperties, rendererIndex=1}
  ]
 }
 Wings {
  emitterUpdateScript [
   &EmitterState(
    Inactive Response=::NewEnumerator1
    Life Cycle Mode=::NewEnumerator1
    Loop Behavior=::NewEnumerator1
    Loop Duration Mode=::NewEnumerator1
    UseLoopDelay=false
   )
   &SpawnBurst_Instantaneous(Use Spawn Probability=false, UseLoopCountLimit=false)
   &SpawnBurst_Instantaneous(Spawn Group=1, Use Spawn Probability=false, UseLoopCountLimit=false) {module=SpawnBurst_Instantaneous001}
  ]
  particleSpawnScript [&InitializeParticle(Lifetime=2, Mesh Scale=<1.3 1 1>, Mesh Scale Mode=::NewEnumerator4, UsePositionOffset=false, Write Material Random=true)]
  particleUpdateScript [
   &ParticleState(Kill Particles When Lifetime Has Elapsed=false, Let Infinitely Lived Particles Die When Emitter Deactivates=true)
   /SetVariables(Particles.Position=/Niagara/DynamicInputs/Transforms/TransformPosition(
     Position=&MaskVectorBySpawnGroup(Masked Vector=<2 -3 -2>, Passthrough Vector=<-2 -3 -2>)
     Source Space=::Local
     Destination Space=::Simulation
    )) {isSetParameters=true}
   &DynamicMaterialParameters(
    Index 0 Param 1=&FloatFromCurve(
     FloatCurve=DataInterface() {propertyValues {Curve {keys [
         {arriveTangent=3.0388, interpMode=RCIM_Cubic, leaveTangent=3.0388, tangentMode=RCTM_User, time=0, value=-0.2}
         {arriveTangent=1.6667, leaveTangent=1.6667, time=0.15, value=1}
         {arriveTangent=-1.2941, interpMode=RCIM_Cubic, leaveTangent=-1.2941, time=0.85, value=1}
         <at 1 -0.2>
        ]}}}
     CurveIndex=$System.NormalizedTimeSinceLastEvent
    )
    Module.Param0WriteEnabled=true
    Module.Param1WriteEnabled=true
    Module.Param2WriteEnabled=true
    Module.Param3WriteEnabled=true
   )
   &InterpolateOverTime(
    Rate Of Change=8
    Target Quat Value:Quat4f=&MultiplyQuaternion(
     Quaternion A=$Engine.Owner.Rotation
     Quaternion B=&MakeQuatFromAxisAngle(Axis=<0 0 1>, AngleInDegrees=&MaskFloatBySpawnGroup(Masked Float=0, Passthrough Float=180))
    )
    Variable Type=::NewEnumerator5
   )
   &self:LaggedRotation(
    CurrentQuaternion:Quat4f=&MultiplyQuaternion(
     Quaternion A=$Engine.Owner.Rotation
     Quaternion B=&MakeQuatFromAxisAngle(Axis=<0 0 1>, AngleInDegrees=&MaskFloatBySpawnGroup(Masked Float=0, Passthrough Float=180))
    )
    LaggedQuaternion:Quat4f=$Particles.InterpolateOverTime.InterpolateOverTime_Quaternion.Moving Average
   ) {module=LaggedRotation}
  ]
  rendererClasses [&NiagaraMeshRendererProperties]
  renderers [{rendererClass=&NiagaraMeshRendererProperties, rendererIndex=0}]
 }
}
summary {
 systemName=NS_SummerStrideBackpack
 emitters [
  {emitterName=NE_SummerStrideGlobs, rendererClasses [&NiagaraMeshRendererProperties]}
  {emitterName=Wings, rendererClasses [&NiagaraMeshRendererProperties]}
  {emitterName=Rays, rendererClasses [&NiagaraMeshRendererProperties, &NiagaraSpriteRendererProperties]}
  {emitterName=LowGlow, rendererClasses [&NiagaraSpriteRendererProperties]}
  {emitterName=Lights, rendererClasses [&NiagaraLightRendererProperties]}
 ]
 userVariables [
  User.EventDuration {type=NiagaraFloat, defaultValue=5, description=""}
  User.KillEvent {type=NiagaraBool}
 ]
}
